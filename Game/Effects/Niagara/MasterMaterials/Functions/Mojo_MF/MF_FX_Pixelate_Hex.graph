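# Material function for FX materials that pixelates texture coordinates into hexagonal cells.
# Centered, scaled UVs are tested against two offset rectangular grids with a 1.732:1 ratio, and a Custom node picks the nearer cell center as the hex ID.
# A second Custom node converts that ID back to UV space using Scale and Stretch.
# It outputs the quantized UVs and the input texture sampled with them, previewing with a noise texture.
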MaterialFunction /Game/Effects/Niagara/MasterMaterials/Functions/Mojo_MF/MF_FX_Pixelate_Hex
graph {nodes [
  n1 = Constant2Vector() {g=0.5, r=1}
  n2 = TextureCoordinate()
  MaterialExpressionReroute_6 = Reroute(None=FunctionInput(Preview=Constant() {desc=scale, r=20}) {inputName=Scale, inputType=FunctionInput_Scalar, sortPriority=2}) {desc=sc}
  n3 = Multiply(
   A=Add(
    A=FunctionInput(Preview=@n2) {bUsePreviewValueAsDefault=true, inputName=In TexCoords, inputType=FunctionInput_Vector2, sortPriority=1}
    B=Constant() {r=-0.5}
   )
   B=@MaterialExpressionReroute_6
  ) {desc=p}
  MaterialExpressionReroute_1 = Reroute(None=AppendVector(A=Constant() {r=1.732}, B=Constant() {r=1})) {desc=s}
  n4 = Divide(A=AppendVector(A=@n3, B=Subtract(A=@n3, B=@n1) {constB=0.5}), B=AppendVector(A=@MaterialExpressionReroute_1, B=@MaterialExpressionReroute_1))
  MaterialExpressionReroute_0 = Reroute(None=Add(A=Floor(None=@n4), B=Constant() {r=0.5})) {desc=hc}
  n5 = Multiply(A=ComponentMask(None=@MaterialExpressionReroute_0) {g=true, r=true}, B=@MaterialExpressionReroute_1)
  n6 = Multiply(A=Add(A=ComponentMask(None=@MaterialExpressionReroute_0) {a=true, b=true}, B=Constant() {r=0.5}), B=@MaterialExpressionReroute_1)
  MaterialExpressionReroute_5 = Reroute(None=AppendVector(A=Subtract(A=@n3, B=@n5), B=Subtract(A=@n3, B=@n6))) {desc=h}
  n7 = Custom!(h=@MaterialExpressionReroute_5, hC=@MaterialExpressionReroute_0) {
   code="return dot(h.xy, h.xy)<dot(h.zw, h.zw) ? float4(h.xy, hC.xy) : float4(h.zw, hC.zw + .5);"
   outputType=CMOT_Float4
   inputs [
    {inputName=h, input {expression=&self:MaterialExpressionReroute_5}}
    {inputName=hC, input {expression=&self:MaterialExpressionReroute_0}}
   ]
  }
  MaterialExpressionComponentMask_3 = ComponentMask(None=@n7) {a=true, b=true}
  MaterialExpressionFunctionInput_3 = FunctionInput(Preview=Constant() {desc=stretch, r=0.7}) {bUsePreviewValueAsDefault=true, inputName=Stretch, inputType=FunctionInput_Scalar, sortPriority=3}
  n8 = Custom!(
   h=@MaterialExpressionComponentMask_3
   s=@MaterialExpressionReroute_1
   sc=@MaterialExpressionReroute_6
   stretch=@MaterialExpressionFunctionInput_3
  ) {
   code="return h*s/sc*stretch;"
   outputType=CMOT_Float2
   inputs [
    {inputName=h, input {expression=&self:MaterialExpressionComponentMask_3}}
    {inputName=s, input {expression=&self:MaterialExpressionReroute_1}}
    {inputName=sc, input {expression=&self:MaterialExpressionReroute_6}}
    {
     inputName=stretch
     input {expression=&self:MaterialExpressionFunctionInput_3}
    }
   ]
  }
  FunctionOutput(None=@n8) {outputName=Out UVs, sortPriority=1}
  n9 = TextureObject() {texture=/Game/Global/Textures/Default/Noise/Noise}
  FunctionOutput(None=TextureSample(UVs=@n8, Tex=FunctionInput(Preview=@n9) {bUsePreviewValueAsDefault=true, inputName=In Texture, inputType=FunctionInput_Texture2D}) {samplerType=SAMPLERTYPE_LinearGrayscale}.RGBA)
 ]}
