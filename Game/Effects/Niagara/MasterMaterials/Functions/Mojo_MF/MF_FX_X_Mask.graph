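# MF_FX_X_Mask is a material function that produces an SDF-style X-shaped mask for Niagara effect materials.
# It turns the difference and sum of the input UV components into two diagonal gradients, each passed through MF_FX_Stroke with a fixed center and width.
# Both strokes are added to the In Color input and saturated into the Mask output, so the X can layer onto an incoming value.
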MaterialFunction /Game/Effects/Niagara/MasterMaterials/Functions/Mojo_MF/MF_FX_X_Mask
graph {
 properties {description=Produces a clean sdf X mask shape}
 nodes [
  n1 = FunctionInput(Preview=TextureCoordinate()) {bUsePreviewValueAsDefault=true, inputName=In TexCoord, inputType=FunctionInput_Vector2}
  n2 = Constant() {r=0.5}
  n3 = Multiply(A=Subtract(A=ComponentMask(None=@n1) {r=true}, B=ComponentMask(None=@n1) {g=true}), B=@n2)
  n4 = Add(
   A=FunctionInput(Preview=Constant3Vector()) {bUsePreviewValueAsDefault=true, inputName=In Color, sortPriority=1}
   B=MaterialFunctionCall!(X=Add(A=@n3, B=@n2), S=Constant() {r=0.5}, W=Constant() {r=0.1}) {materialFunction=&/MF_FX_Stroke}.Result
  )
  n5 = ComponentMask(None=@n1) {r=true}
  n6 = ComponentMask(None=@n1) {g=true}
  n7 = MaterialFunctionCall(X=Multiply(A=Add(A=@n5, B=@n6), B=Constant() {r=0.5}), S=Constant() {r=0.5}, W=Constant() {r=0.1}) {materialFunction=&MF_FX_Stroke}
  FunctionOutput(None=Saturate(None=Add(A=@n4, B=@n7.Result))) {outputName=Mask}
 ]
}
