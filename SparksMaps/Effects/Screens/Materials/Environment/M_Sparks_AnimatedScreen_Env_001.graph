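# Lit environment material for an animated tech-style screen with a blue default palette and emissive glow.
# SDF masks build indented frames and procedural details such as circles, loading dots, a sine wave, fake FFT EQ bars and an X icon.
# Quantized UVs drive a radial gradient and flipbook, alongside LCD and noise-driven hex or square pattern layers.
# Fresnel falloff, vignette and scanlines shape the result feeding base color and emissive.
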Material /SparksMaps/Effects/Screens/Materials/Environment/M_Sparks_AnimatedScreen_Env_001
graph {
 properties {bHasPixelAnimation=true, bUseMaterialAttributes=true, bUsedWithNanite=true, pixelDepthOffsetMode=PDOM_Legacy, usedShadingModels=Default Lit}
 nodes [
  n1 = Multiply(A=Time(), B=ScalarParameter() {defaultValue=0.0001, parameterName=Backlight Noise Speed Mult})
  Abs(None=Sine(None=@n1))
  Constant2Vector() {g=0.5, r=0.75}
  Constant3Vector() {constant=<rgba 0.0082 0.4045 1 1>}
  Constant() {r=0.1}
  Constant() {r=0.5}
  Constant() {r=0.1}
  Constant() {r=1}
  Constant()
  n2 = Add(A=Multiply(A=Time()) {constB=-1}) {constB=5}
  n3 = Constant2Vector() {g=0.0856, r=0.0856}
  n4 = Add(A=Multiply(A=TextureCoordinate(), B=Constant2Vector() {g=2, r=2}), B=Constant2Vector() {g=-1.28})
  n5 = Multiply(
   A=ComponentMask(None=Panner(Coordinate=@n4, Time=@n2, Speed=@n3)) {desc=X, r=true}
   B=MaterialFunctionCall(Multiplier=Constant() {r=1}) {materialFunction=&Pi}.Result
  )
  Cosine(None=@n5)
  n6 = Multiply(A=Time(), B=ScalarParameter() {defaultValue=0.1, parameterName=Speed})
  Frac(None=@n6)
  MaterialFunctionCall() {materialFunction=/FortUILibrary/MaterialFunctions/Basics/MF_UI_WIP_9SliceUV}
  MaterialFunctionCall() {materialFunction=&MF_UI_WIP_9SliceUV}
  n7 = Add(A=Multiply(A=TextureCoordinate(), B=Constant2Vector() {g=2, r=2.55}), B=Constant2Vector() {g=-0.15, r=-1.55})
  MaterialFunctionCall!(In TexCoords=@n7, Time=Constant(), Speed X, Speed Y, Size=Constant() {r=0.25}, Divide Control) {materialFunction=/Game/Effects/Niagara/MasterMaterials/Functions/Mojo_MF/MF_FX_Tunnel_UV}
  MaterialFunctionCall!(In TexCoord, In Color) {materialFunction=/Game/Effects/Niagara/MasterMaterials/Functions/Mojo_MF/MF_FX_X_Mask}
  n8 = Add(A=Multiply(A=TextureCoordinate()) {constB=8}) {constB=-3.25}
  n9 = Constant2Vector() {g=1.25, r=1}
  MaterialFunctionCall!(HalfSize, CornerRadius=Constant() {r=0.01}, Position=Add(A=Multiply(A=@n8, B=@n9), B=Constant2Vector() {g=-2.8, r=1})) {materialFunction=/Game/Effects/Niagara/MasterMaterials/Functions/Mojo_MF/MF_FX_RoundedBox_SDF}
  MaterialFunctionCall!(In TexCoords, Width) {materialFunction=/Game/Effects/Niagara/MasterMaterials/Functions/Mojo_MF/MF_FX_SquareEdgeMask}
  n10 = ScalarParameter() {defaultValue=100, parameterName=Quantize}
  n11 = Multiply(A=TextureCoordinate() {uTiling=0.95, vTiling=0.95}, B=@n10)
  n12 = Divide(A=Round(None=@n11), B=@n10)
  n13 = Multiply(A=MaterialFunctionCall(UVs=@n12, Radius=Constant() {r=0.5}, Density=Constant() {r=0.5}) {materialFunction=&RadialGradientExponential}.RadialGradientExponential) {constB=1.5, desc=Boost Brightness of Radial Grad}
  n14 = NamedRerouteUsage()
  Max(A=@n13, B=LinearInterpolate(Alpha=@n14) {constB=0.75}) {desc=Add Tech-y Frame Mask}
  n15 = NamedRerouteUsage()
  n16 = ComponentMask(None=TextureCoordinate()) {r=true}
  n17 = NamedRerouteUsage()
  n18 = NamedRerouteUsage()
  n19 = Divide(A=ComponentMask(None=@n18) {g=true}, B=ComponentMask(None=@n18) {r=true})
  n20 = LinearInterpolate(
   A=ScalarParameter() {defaultValue=0.03, parameterName=SpaceBetweenFrames}
   B=ScalarParameter() {defaultValue=0.09, parameterName=SpaceBetweenFrames_Modal}
   Alpha=NamedRerouteUsage()
  )
  n21 = ComponentMask(None=Divide(A=NamedRerouteUsage(), B=NamedRerouteUsage())) {g=true}
  n22 = AppendVector(A=Multiply(A=Multiply(A=@n19, B=@n20), B=@n21), B=Multiply(A=@n20, B=@n21))
  n23 = MaterialFunctionCall(
   UVs=MaterialFunctionCall(UVs=AppendVector(A=@n16, B=@n17), Texture Scale=OneMinus(None=@n22)) {materialFunction=&ScaleUVsByCenter}.UVs
   Size=NamedRerouteUsage()
   CornerRoundness=ScalarParameter() {defaultValue=0.025, parameterName=InnerFrameCornerRoundness}
   OutlineThickness=ScalarParameter() {defaultValue=0.005, parameterName=InnerFrameThickness}
  ) {materialFunction=&MF_UI_Box}
  n24 = OneMinus(None=SmoothStep(Max=ScalarParameter() {defaultValue=0.0125, parameterName=InnerFrameSharpness, sliderMax=1}, Value=@n23.SDFOutline) {constMax=0.0125})
  Multiply(A=OneMinus(None=Multiply(A=@n15, B=@n24)), B=NamedRerouteUsage())
  n25 = Cosine(None=Add(A=Time()))
  Multiply(A=Multiply(A=Add(A=@n25)) {constB=0.5}) {constB=0.4}
  n26 = TextureCoordinate()
  NamedRerouteDeclaration(None=TextureSample(UVs=Multiply(A=@n26, B=@n10)) {samplerType=SAMPLERTYPE_Masks, texture=/Game/Environments/Asteria/Sets/NeonCity/Techart/Textures/T_LCD_Panels_LCDSoft}.RGB) {name=LCD pattern, nodeColor=<rgba 0.5703 1 0 1>}
  n27 = Constant() {r=-3}
  n28 = MaterialFunctionCall(Value=@n6) {materialFunction=&LinearSine}
  n29 = MaterialFunctionCall!(
   Texture=TextureObjectParameter() {parameterName=Screen Details Flipbook, texture=/Game/Athena/Environments/Textures/T_NeonCat_MonitorB}
   UVs=@n12
   SubImages=Constant2Vector() {g=3, r=1}
   Frame=Clamp(None=LinearInterpolate(A=@n27, Alpha=@n28.Linear Sine) {constA=-10, constB=5}) {maxDefault=24}
  ) {materialFunction=/Engine/Functions/Engine_MaterialFunctions03/Texturing/SubUV_Function}
  NamedRerouteDeclaration(None=@n29.RGB) {name=LCD RGB, nodeColor=<rgba 1 0.7734 0 1>}
  NamedRerouteDeclaration(None=@n23.Fill) {name=InnerFrameFIll, nodeColor=<rgba 0 0 0 1>}
  NamedRerouteDeclaration(None=Constant2Vector() {g=0.7, r=0.9}) {name=DefaultFrameSize, nodeColor=<rgba 0 0 0 1>}
  n30 = TextureCoordinate()
  n31 = ComponentMask(None=@n30) {g=true}
  n32 = Saturate(None=Subtract(A=@n31, B=ScalarParameter() {defaultValue=0.025, group=Top, parameterName=IndentHeight_Top, sliderMax=0.1, sliderMin=-0.1, sortPriority=2}))
  n33 = ScalarParameter() {defaultValue=0.34, group=Top, parameterName=IndentWidth_Top, sliderMax=0.51, sliderMin=0.175, sortPriority=0}
  n34 = Subtract(A=@n33, B=ScalarParameter() {defaultValue=0.02, group=Top, parameterName=IndentTightness_Top, sliderMax=1, sliderMin=0.01, sortPriority=1})
  n35 = OneMinus(None=@n33)
  n36 = OneMinus(None=@n34)
  n37 = SmoothStep(Value=Add(
    A=MaterialFunctionCall(GradientStart=@n33, GradientEnd=@n34) {materialFunction=&MF_UI_LinearGradient}.Result
    B=MaterialFunctionCall(GradientStart=@n35, GradientEnd=@n36) {materialFunction=&MF_UI_LinearGradient}.Result
   ))
  n38 = ComponentMask(None=@n30) {g=true}
  n39 = ScalarParameter() {defaultValue=0.0375, group=Bottom, parameterName=IndentHeight_Bottom, sliderMax=0.1, sliderMin=-0.1, sortPriority=2}
  n40 = Multiply(A=@n39) {constB=-1}
  n41 = NamedRerouteUsage()
  n42 = Saturate(None=Subtract(A=@n38, B=LinearInterpolate(A=@n40, B=@n39, Alpha=@n41) {constAlpha=0}))
  n43 = ScalarParameter() {defaultValue=0.39, group=Bottom, parameterName=IndentWidth_Bottom, sliderMax=0.51, sliderMin=0.175, sortPriority=0}
  n44 = ScalarParameter() {defaultValue=0.02, group=Bottom, parameterName=IndentTightness_Bottom, sliderMax=1, sliderMin=0.01, sortPriority=1}
  n45 = Subtract(A=@n43, B=@n44)
  n46 = OneMinus(None=@n43)
  n47 = OneMinus(None=@n45)
  n48 = SmoothStep(Value=Add(
    A=MaterialFunctionCall(GradientStart=@n43, GradientEnd=@n45) {materialFunction=&MF_UI_LinearGradient}.Result
    B=MaterialFunctionCall(GradientStart=@n46, GradientEnd=@n47) {materialFunction=&MF_UI_LinearGradient}.Result
   ))
  n49 = ComponentMask(None=TextureCoordinate() {coordinateIndex=4}) {g=true}
  n50 = Saturate(None=Ceil(None=Subtract(A=@n49) {constB=0.5}))
  NamedRerouteDeclaration(None=LinearInterpolate(A=LinearInterpolate(A=@n32, B=@n31, Alpha=@n37), B=LinearInterpolate(A=@n42, B=@n38, Alpha=@n48), Alpha=@n50)) {name=InnerFrameGreenChannelUVs, nodeColor=<rgba 0 0 0 1>}
  NamedRerouteDeclaration(None=MaterialFunctionCall("Value (0-1)"=ScalarParameter() {parameterName=ModalInnerFrameZoom, sliderMax=1}) {materialFunction=&MF_UI_EaseCurves}.EaseInOutQuint) {name=ModalInnerFrameZoom, nodeColor=<rgba 0.5703 1 0 1>}
  n51 = Constant() {r=0.35}
  n52 = Add(A=@n51) {constB=0.1}
  n53 = Max(
   A=MaterialFunctionCall(UVs=@n7, Size=@n52, OutlineThickness=Constant() {r=0.045}) {materialFunction=&MF_UI_Circle}.Outline
   B=Saturate(None=NamedRerouteUsage())
  )
  n54 = Add(A=Multiply(A=TextureCoordinate(), B=Constant2Vector() {g=2, r=2.55}), B=Constant2Vector() {g=-0.15, r=-0.9})
  n55 = Constant() {r=0.05}
  n56 = Max(
   A=Multiply(
    A=MaterialFunctionCall(UVs=@n54, Size=@n52, OutlineThickness=@n55) {materialFunction=&MF_UI_Circle}.Outline
    B=NamedRerouteUsage()
   )
   B=MaterialFunctionCall(UVs=@n54, Size=@n51, OutlineThickness=Constant() {r=0.025}) {materialFunction=&MF_UI_Circle}.Outline
  )
  n57 = Add(A=Add(A=Add(A=@n53, B=@n56), B=NamedRerouteUsage()), B=NamedRerouteUsage())
  NamedRerouteDeclaration(None=Saturate(None=Multiply(A=@n57) {desc=optional darken})) {name=ScreenDetails, nodeColor=<rgba 0 0.2031 1 1>}
  NamedRerouteDeclaration(None=AppendVector(A=ScalarParameter() {defaultValue=0.965, parameterName=FrameWidth}, B=ScalarParameter() {defaultValue=0.9, parameterName=FrameHeight})) {name=BaseFrameSize, nodeColor=<rgba 0 0 0 1>}
  NamedRerouteDeclaration(None=AppendVector(
    A=ScalarParameter() {defaultValue=0.9325, parameterName=FrameWidth_Modal}
    B=ScalarParameter() {defaultValue=0.7731, parameterName=FrameHeight_Modal}
   )) {name=ModalFrameSize, nodeColor=<rgba 0 0 0 1>}
  n58 = Constant() {r=1}
  n59 = Constant2Vector() {g=3, r=2}
  n60 = OneMinus(None=MaterialFunctionCall!(
    UV TexCoord=MaterialFunctionCall!(Center, Maintain Aspect Ratio=@n58, Tiling=@n59, Offset, In Tex Coords=@n54) {materialFunction=/Game/Effects/Fort_Effects/MaterialFunctions/MF_Uniform_Radial_UV}.Out UV
    Time
    Speed=Constant() {r=0.05}
    Spin
   ) {materialFunction=/Game/Effects/Niagara/MasterMaterials/Functions/Mojo_MF/MF_FX_Swirl_UV}.Out UV TexCoord)
  n61 = Multiply(A=Power(Base=ComponentMask(None=@n60) {r=true}) {constExponent=4}) {constB=4}
  n62 = Constant() {desc=Min Dot Size}
  n63 = Constant() {desc=Max Dot Size, r=1}
  NamedRerouteDeclaration(None=LinearInterpolate(A=@n62, B=@n63, Alpha=Saturate(None=@n61))) {name=LoadingRadialGrad, nodeColor=<rgba 1 0 0.3516 1>}
  NamedRerouteDeclaration(None=MaterialFunctionCall("Value (0-1)"=ScalarParameter() {defaultValue=1.45, parameterName=MainToModal, sliderMax=1}) {materialFunction=&MF_UI_EaseCurves}.EaseInOutQuint) {name=MainToModal, nodeColor=<rgba 0.5703 1 0 1>}
  NamedRerouteDeclaration(None=@n29.Alpha) {name=LCD Alpha, nodeColor=<rgba 1 0.2344 0 1>}
  n64 = TextureCoordinate()
  n65 = Constant2Vector() {g=0.05, r=0.05}
  n66 = Add(
   A=Multiply(
    A=MaterialFunctionCall(UVs=@n64, Texture Scale=@n65) {materialFunction=&ScaleUVsByCenter}.UVs
    B=Constant2Vector() {g=1, r=1}
   )
   B=Constant2Vector() {g=6, r=8}
  )
  n67 = Constant() {r=0.5}
  n68 = Multiply(A=Subtract(A=ComponentMask(None=@n66) {r=true}, B=ComponentMask(None=@n66) {g=true}), B=@n67)
  n69 = Constant() {r=0.5}
  n70 = Constant() {r=0.1}
  n71 = Add(A=ComponentMask(None=@n66) {r=true}, B=ComponentMask(None=@n66) {g=true})
  n72 = Constant() {r=0.5}
  n73 = Add(
   A=MaterialFunctionCall!(X=Add(A=@n68, B=@n67), S=@n69, W=@n70) {materialFunction=/Game/Effects/Niagara/MasterMaterials/Functions/Mojo_MF/MF_FX_Stroke}.Result
   B=MaterialFunctionCall(X=Multiply(A=@n71, B=@n72), S=@n69, W=@n70) {materialFunction=&MF_FX_Stroke}.Result
  )
  n74 = Saturate(None=@n66)
  n75 = Constant() {r=0.1}
  NamedRerouteDeclaration(None=Multiply(
    A=Saturate(None=@n73)
    B=MaterialFunctionCall(In TexCoords=@n74, Width=@n75) {materialFunction=&MF_FX_SquareEdgeMask}.Mask One Minus
   )) {name=Close Icon X, nodeColor=<rgba 1 0.0938 0 1>}
  n76 = MaterialFunctionCall() {materialFunction=&MF_UI_WIP_9SliceUV}
  n77 = ComponentMask(None=@n76.Result) {g=true}
  n78 = ScalarParameter() {defaultValue=0.025, group=Top, parameterName=IndentHeight_Top_Outer, sliderMax=0.1, sliderMin=-0.1, sortPriority=2}
  n79 = Saturate(None=Subtract(A=@n77, B=@n78))
  n80 = ScalarParameter() {defaultValue=0.34, group=Top, parameterName=IndentWidth_Top_Outer, sliderMax=0.51, sliderMin=0.175, sortPriority=0}
  n81 = ScalarParameter() {defaultValue=0.02, group=Top, parameterName=IndentTightness_Top_Outer, sliderMax=1, sliderMin=0.01, sortPriority=1}
  n82 = Subtract(A=@n80, B=@n81)
  n83 = OneMinus(None=@n80)
  n84 = OneMinus(None=@n82)
  n85 = SmoothStep(Value=Add(
    A=MaterialFunctionCall(GradientStart=@n80, GradientEnd=@n82) {materialFunction=&MF_UI_LinearGradient}.Result
    B=MaterialFunctionCall(GradientStart=@n83, GradientEnd=@n84) {materialFunction=&MF_UI_LinearGradient}.Result
   ))
  n86 = ComponentMask(None=@n76.Result) {g=true}
  n87 = ScalarParameter() {defaultValue=0.0375, group=Bottom, parameterName=IndentHeight_Bottom_Outer, sliderMax=0.1, sliderMin=-0.1, sortPriority=2}
  n88 = Multiply(A=@n87) {constB=-1}
  n89 = NamedRerouteUsage()
  n90 = Saturate(None=Subtract(A=@n86, B=LinearInterpolate(A=@n88, B=@n87, Alpha=@n89) {constAlpha=0}))
  n91 = ScalarParameter() {defaultValue=0.39, group=Bottom, parameterName=IndentWidth_Bottom_Outer, sliderMax=0.51, sliderMin=0.175, sortPriority=0}
  n92 = ScalarParameter() {defaultValue=0.02, group=Bottom, parameterName=IndentTightness_Bottom_Outer, sliderMax=1, sliderMin=0.01, sortPriority=1}
  n93 = Subtract(A=@n91, B=@n92)
  n94 = OneMinus(None=@n91)
  n95 = OneMinus(None=@n93)
  n96 = SmoothStep(Value=Add(
    A=MaterialFunctionCall(GradientStart=@n91, GradientEnd=@n93) {materialFunction=&MF_UI_LinearGradient}.Result
    B=MaterialFunctionCall(GradientStart=@n94, GradientEnd=@n95) {materialFunction=&MF_UI_LinearGradient}.Result
   ))
  NamedRerouteDeclaration(None=LinearInterpolate(A=LinearInterpolate(A=@n79, B=@n77, Alpha=@n85), B=LinearInterpolate(A=@n90, B=@n86, Alpha=@n96), Alpha=@n50)) {name=OuterFrameGreenChannelUVs, nodeColor=<rgba 0 0 0 1>}
  n97 = Add(A=Constant() {desc=Height Offset 1, r=0.1}, B=Multiply(A=Sine(None=@n5) {desc=flip direction}, B=Constant() {r=0.15}))
  n98 = ComponentMask(None=@n4) {desc=X, g=true}
  n99 = Constant() {desc=Width 1, r=0.05}
  n100 = Multiply(A=TextureCoordinate(), B=Constant2Vector() {g=1, r=0.75})
  n101 = Constant2Vector() {g=-0.18, r=0.3}
  n102 = MaterialFunctionCall!(st=Saturate(None=Add(A=@n100, B=@n101)), s) {materialFunction=/Game/Effects/Niagara/MasterMaterials/Functions/Mojo_MF/MF_FX_Rect_SDF}
  n103 = Constant() {desc=S, r=0.25}
  n104 = MaterialFunctionCall!(x=@n102.Result, size=@n103) {materialFunction=/Game/Effects/Niagara/MasterMaterials/Functions/Mojo_MF/MF_FX_Fill}
  n105 = Multiply(
   A=Multiply(A=MaterialFunctionCall!(X=@n98, S=@n97, W=@n99) {materialFunction=/Game/Effects/Fort_Effects/MaterialFunctions/MF_Stroke}.Result) {constB=0.8}
   B=@n104.Result
  )
  n106 = MaterialFunctionCall(X=@n102.Result, S=@n103, W=Constant() {desc=W, r=0.01}) {materialFunction=&MF_FX_Stroke}
  n107 = Add(A=@n100, B=Constant2Vector() {g=-0.2, r=-0.05})
  n108 = Constant() {r=0.3}
  n109 = MaterialFunctionCall(st=@n107) {materialFunction=&MF_FX_Rect_SDF}
  n110 = Constant() {desc=S, r=0.15}
  n111 = Multiply(
   A=MaterialFunctionCall!(
    UV=MaterialFunctionCall(UVs=@n107, Texture Scale=@n108) {materialFunction=&ScaleUVsByCenter}.UVs
    Square Grid Width
    RandomizeUVPlacement
    RadialUV
    BarAmount Vert
    BarAmount
    BarWidth
    MountainHardness
    FFT
    Use Fake FFT Noise=StaticBool() {value=true}
    Fake FFT Noise Speed X=Constant() {r=0.075}
    Fake FFT Noise Pow=Constant() {r=4}
    FFT_IntensityMult
    FFT_Add
   ) {materialFunction=/SparksMaps/Effects/Screens/Materials/MF/MF_FX_FFT_EQ}.EQ Bars
   B=MaterialFunctionCall(x=@n109.Result, size=@n110) {materialFunction=&MF_FX_Fill}.Result
  )
  n112 = MaterialFunctionCall(X=@n109.Result, S=@n110, W=Constant() {desc=W, r=0.01}) {materialFunction=&MF_FX_Stroke}
  NamedRerouteDeclaration(None=Add(A=Add(A=@n105, B=@n106.Result), B=Add(A=@n111, B=@n112.Result))) {name=sineWaveBoxMask, nodeColor=<rgba 0 1 0.4375 1>}
  n113 = Constant() {desc="Random Tiling ", r=16}
  n114 = Multiply(
   A=MaterialFunctionCall(Maintain Aspect Ratio=Constant() {r=1}, Tiling=Constant2Vector() {g=3, r=2}, In Tex Coords=@n7) {materialFunction=&MF_Uniform_Radial_UV}.Out UV
   B=@n113
  )
  n115 = MaterialFunctionCall!(In Texcoords=Frac(None=@n114)) {materialFunction=/Game/Effects/Fort_Effects/MaterialFunctions/MF_Circle_SDF}
  n116 = OneMinus(None=MaterialFunctionCall(UV TexCoord=@n7, Speed=Constant() {r=0.25}) {materialFunction=&MF_FX_Swirl_UV}.Out UV TexCoord)
  n117 = Multiply(A=Power(Base=ComponentMask(None=@n116) {r=true}) {constExponent=4}) {constB=4}
  n118 = MaterialFunctionCall(In Texcoords=@n7) {materialFunction=&MF_Circle_SDF}
  n119 = MaterialFunctionCall!(x=@n118.Result, size=Constant() {r=0.25}) {materialFunction=/Game/Effects/Fort_Effects/MaterialFunctions/MF_Fill}
  n120 = Constant() {r=0.5}
  n121 = Saturate(None=Multiply(
    A=OneMinus(None=@n119.Result)
    B=MaterialFunctionCall(x=@n118.Result, size=@n120) {materialFunction=&MF_Fill}.Result
   ))
  n122 = Multiply(A=LinearInterpolate(A=@n62, B=@n63, Alpha=Saturate(None=@n117)), B=@n121) {desc=Circle SDF Thickness}
  NamedRerouteDeclaration(None=Multiply(A=MaterialFunctionCall(x=@n115.Result, size=@n122) {materialFunction=&MF_Fill}.Result, B=@n121)) {name=loadingDotMask, nodeColor=<rgba 0 1 0.9766 1>}
  n123 = VectorParameter() {parameterName=Color Light, defaultValue=<rgba 0.0027 0.2789 0.7305 1>}
  n124 = NamedRerouteUsage()
  n125 = NamedRerouteUsage()
  n126 = ScalarParameter() {defaultValue=1, desc=0-1 Values, parameterName=Screen Details Lerp, sliderMax=1}
  n127 = LinearInterpolate(
   A=LinearInterpolate(
    A=VectorParameter() {parameterName=Color Dark, defaultValue=<rgba 0.006 0.0382 0.3515 1>}.RGB
    B=Add(A=@n123.RGB, B=@n124)
    Alpha=Saturate(None=@n13)
   )
   B=VectorParameter() {parameterName=Frame Color, defaultValue=<rgba 0.0009 0.5395 1 1>}.RGB
   Alpha=Max(A=Multiply(A=@n14) {constB=2}, B=LinearInterpolate(B=@n125, Alpha=@n126)) {desc=Layer In screen details}
  )
  n128 = LinearInterpolate(Alpha=NamedRerouteUsage()) {constA=0.25, constB=0.5}
  NamedRerouteDeclaration(None=Multiply(A=Multiply(A=@n127, B=@n128) {desc=darken outside the main window frame}) {constB=0.5, desc=Overall Darken}) {name=PrimaryScreenColor, nodeColor=<rgba 1 0 0.8672 1>}
  n129 = Multiply(A=Time(), B=ScalarParameter() {defaultValue=0.1, parameterName=BlockSpeed})
  n130 = Multiply(A=@n11) {constB=0.15}
  n131 = MaterialFunctionCall!(
   Time Mult=Constant() {r=0.25}
   Time Add=MaterialFunctionCall!(In Texture, In TexCoords=@n130, Scale=Constant() {r=1}, Stretch) {materialFunction=/Game/Effects/Niagara/MasterMaterials/Functions/Mojo_MF/MF_FX_Pixelate_Hex}.Result
  ) {materialFunction=/Game/Effects/Niagara/MasterMaterials/Functions/Mojo_MF/MF_FX_Flicker}
  n132 = LinearInterpolate(
   A=LinearInterpolate(Alpha=MaterialFunctionCall!(PrimaryTextureCoordinates=@n130, 2ndOrderTextureCoordinates=@n130, RandomSeed=@n129, NoiseDensityBias, Bypass Noise Influence Crosstalk) {materialFunction=/Game/Effects/Fort_Effects/Materials/_Functions/MF_RandomUVSample}.NoiseWithRange) {constA=0.5}
   B=LinearInterpolate(Alpha=Saturate(None=@n131.Raw Flicker)) {constA=0.5}
   Alpha=ScalarParameter() {defaultValue=1, parameterName=Use Hex Random Pattern?}
  )
  n133 = Multiply(A=TextureCoordinate(), B=ScalarParameter() {defaultValue=0.25, parameterName=Backlight Noise Tile}) {constB=0.25}
  n134 = LinearInterpolate(
   A=Constant2Vector() {g=0.01, r=-0.01}
   B=Constant2Vector() {g=-0.01, r=0.01}
   Alpha=MaterialFunctionCall(Value=@n1) {materialFunction=&LinearSine}.Linear Sine
  )
  n135 = TextureSample(UVs=Panner(Coordinate=TextureCoordinate() {uTiling=0.5, vTiling=0.5}) {speedY=-0.02}) {texture=&2ndInfectedMapLinear}
  n136 = TextureSample(UVs=Add(A=Panner(Coordinate=@n133, Speed=@n134) {speedX=-0.1, speedY=0.1}, B=Multiply(A=@n135.R) {constB=0.025}) {desc=Distort UVs}) {samplerType=SAMPLERTYPE_LinearGrayscale, texture=/Game/Global/Textures/Default/Noise/LowResBlurredNoise_NoMips}
  n137 = LinearInterpolate(Alpha=Saturate(None=Power(Base=@n136.G))) {constA=0.1, constB=1.5}
  NamedRerouteDeclaration(None=LinearInterpolate(Alpha=Multiply(A=@n137, B=@n132)) {constA=0.1}) {name=Random Square or Hex Patterns, desc=Swapped squares to Hex, nodeColor=<rgba 0 0.1328 1 1>}
  n138 = Constant4Vector() {constant=<rgba 0.0885 0.0885 0.0885 1>}
  n139 = ComponentMask(None=@n138.A) {a=true}
  n140 = AppendVector(A=@n16, B=@n17)
  n141 = NamedRerouteUsage()
  n142 = ScalarParameter() {defaultValue=0.05, parameterName=OuterFrameCornerRoundness}
  n143 = ScalarParameter() {defaultValue=0.001, parameterName=OuterFrameThickness}
  n144 = OneMinus(None=SmoothStep(
    Max=ScalarParameter() {defaultValue=0.0041, parameterName=OuterFrameSharpness, sliderMax=1}
    Value=MaterialFunctionCall(UVs=@n140, Size=@n141, CornerRoundness=@n142, OutlineThickness=@n143) {materialFunction=&MF_UI_Box}.SDFOutline
   ) {constMax=0.0041})
  n145 = Constant4Vector() {constant=<rgba 1 1 1 1>}
  n146 = ComponentMask(None=@n145.A) {a=true}
  NamedRerouteDeclaration(None=Max(A=Multiply(A=@n139, B=@n144) {constB=0.25}, B=Multiply(A=@n146, B=@n24) {constB=0.7})) {name=FrameMask, nodeColor=<rgba 0 0 0 1>}
  n147 = ComponentMask(None=@n138.A) {b=true, g=true, r=true}
  n148 = Ceil(None=@n144)
  n149 = ComponentMask(None=@n145.A) {b=true, g=true, r=true}
  n150 = Max(A=Ceil(None=@n24), B=@n15)
  NamedRerouteDeclaration(None=Add(A=Multiply(A=@n147, B=@n148) {constB=0.25}, B=Multiply(A=@n149, B=@n150) {constB=0.25})) {name=FrameColors, nodeColor=<rgba 0 0 0 1>}
  NamedRerouteDeclaration(None=LinearInterpolate(A=NamedRerouteUsage(), B=NamedRerouteUsage(), Alpha=NamedRerouteUsage()) {constAlpha=1}) {name=CurrentFrameSize, nodeColor=<rgba 0 0 0 1>}
  NamedRerouteUsage()
  OneMinus(None=@n104.Result)
  Reroute(None=Divide(A=Floor(None=@n114), B=@n113))
  Reroute(None=@n13)
  Saturate(None=Add(A=@n51) {constB=-0.25})
  n151 = TextureCoordinate() {uTiling=2, vTiling=2}
  n152 = OneMinus(None=Fresnel() {baseReflectFraction=0, exponent=2})
  n153 = Multiply(A=NamedRerouteUsage()) {constB=0.15}
  n154 = NamedRerouteUsage()
  n155 = Multiply(A=NamedRerouteUsage(), B=LinearInterpolate(A=NamedRerouteUsage(), B=NamedRerouteUsage(), Alpha=Multiply(A=@n153, B=@n154)) {constAlpha=0.1})
  n156 = NamedRerouteUsage()
  n157 = Multiply(
   A=LinearInterpolate(
    A=Multiply(A=@n155, B=@n156)
    B=VectorParameter() {parameterName=WashColor, defaultValue=<rgba 0.0009 0.5333 0.9734 1>}.RGB
    Alpha=ScalarParameter() {defaultValue=0.004, parameterName=WashIntensity}
   )
   B=ScalarParameter() {defaultValue=20, parameterName=Brightness}
  )
  n158 = Power(Base=MaterialFunctionCall!(TexCoords, Edge Sharpness=Constant() {r=5}, Edge Subtract) {materialFunction=/Game/Effects/Niagara/MasterMaterials/Functions/Mojo_MF/MF_FX_UV_EdgeMask}.1-Edge Mask) {constExponent=5.5}
  n159 = Constant() {r=500}
  n160 = Multiply(
   A=Multiply(A=Multiply(A=@n152, B=@n157), B=Multiply(A=@n158) {constB=1.15}) {desc=Vignette}
   B=LinearInterpolate(Alpha=MaterialFunctionCall!(Scanline Amount=@n159) {materialFunction=/Game/Environments/Asteria/Sets/NeonCity/Techart/Holograms/NewHologramMaterial/MFScanlines}.Result) {constA=0.5}
  ) {desc=Scanlines}
  SetMaterialAttributes!(
   MaterialAttributes
   Base Color=@n160
   Emissive Color=MaterialFunctionCall(Emissive=@n160) {materialFunction=&MF_FX_EmissiveAdjustments}.Result
   Roughness=LinearInterpolate(Alpha=TextureSample(UVs=@n151) {samplerType=SAMPLERTYPE_LinearGrayscale, texture=/Game/Global/Textures/Default/Noise/EqualizerComponents_v2_Noise}.G) {constB=0.25}
  ) {attributeSetTypes [#1, #2, #3]}
  TextureCoordinate()
  TextureCoordinate()
  TextureCoordinate()
  TextureCoordinate()
  TextureObject() {texture=/Game/Environments/Asteria/Sets/NeonCity/Techart/NeonSigns/SpecialAnimatedSigns/Animation/FaceSequence}
 ]
}
