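# UI-domain material that draws an animated LCD-style screen, apparently an environment screen variant for Sparks maps.
# Sine scanlines on 9-slice UI UVs blend two tint colors into a primary screen color, alongside an LCD panel mask texture.
# A backlight layer uses panning noise with distorted UVs, modulated by a pattern lerped between random squares and flickering hexes.
# Named reroutes share these layers, with brightness and edge-mask vignette shaping the image.
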Material /SparksMaps/Effects/Screens/Materials/Environment/M_Sparks_AnimatedScreen_Env_UI
graph {
 properties {bHasPixelAnimation=true, bUsedWithNanite=true, materialDomain=MD_UI, pixelDepthOffsetMode=PDOM_Legacy, usedShadingModels=Default Lit}
 nodes [
  n1 = Multiply(A=Time(), B=ScalarParameter() {defaultValue=0.0001, parameterName=Backlight Noise Speed Mult})
  Abs(None=Sine(None=@n1))
  Constant3Vector() {constant=<rgba 0 0.0615 1 1>}
  n2 = MaterialFunctionCall() {materialFunction=&GetUserInterfaceUV}
  n3 = Sine(None=Multiply(A=ComponentMask(None=@n2.9-Slice UV) {g=true}, B=ScalarParameter() {defaultValue=25, group=Screen Details, parameterName=Scanlines Tile}) {constB=125})
  n4 = ScalarParameter() {defaultValue=0.5, group=Screen Details, parameterName=Scanelines Intensity}
  n5 = Saturate(None=Add(A=ConstantBiasScale(None=@n3), B=OneMinus(None=@n4)) {constB=0.25})
  LinearInterpolate(
   A=@n5
   Alpha=MaterialFunctionCall(Fade Length=ScalarParameter() {defaultValue=500, group=Screen Details, parameterName=Scanlines Camera Fade}) {materialFunction=&CameraDepthFade}.Result
  ) {constA=1}
  MaterialFunctionCall() {materialFunction=&MF_UI_UVScaling}
  MaterialFunctionCall() {materialFunction=&GetUserInterfaceUV}
  n6 = Constant() {r=500}
  n7 = NamedRerouteUsage()
  n8 = NamedRerouteUsage()
  n9 = Constant() {r=0.65}
  n10 = Multiply(
   A=Multiply(A=Multiply(A=@n7) {constB=0.15}, B=LinearInterpolate(A=@n8, Alpha=@n9))
   B=ScalarParameter() {defaultValue=7, parameterName=Brightness}
  )
  Multiply(
   A=@n10
   B=LinearInterpolate(Alpha=MaterialFunctionCall!(Scanline Amount=@n6) {materialFunction=/Game/Environments/Asteria/Sets/NeonCity/Techart/Holograms/NewHologramMaterial/MFScanlines}.Result) {constA=0.5, constB=0.75}
  ) {desc=Scanlines}
  n11 = ComponentMask(None=@n2.9-Slice UV) {g=true}
  n12 = Saturate(None=@n11)
  Multiply(A=@n12, B=OneMinus(None=NamedRerouteUsage()))
  Multiply(A=NamedRerouteUsage()) {constB=50}
  n13 = Multiply(A=Time(), B=ScalarParameter() {defaultValue=0.1, parameterName=BlockSpeed})
  n14 = Constant2Vector() {g=100, r=200}
  n15 = MaterialFunctionCall() {materialFunction=&GetUserInterfaceUV}
  n16 = Multiply(A=Multiply(A=@n15.9-Slice UV, B=@n14)) {constB=0.15}
  n17 = MaterialFunctionCall!(
   Time Mult=Constant() {r=0.125}
   Time Add=MaterialFunctionCall!(In Texture, In TexCoords=@n16, Scale=Constant() {r=1}, Stretch) {materialFunction=/Game/Effects/Niagara/MasterMaterials/Functions/Mojo_MF/MF_FX_Pixelate_Hex}.Result
  ) {materialFunction=/Game/Effects/Niagara/MasterMaterials/Functions/Mojo_MF/MF_FX_Flicker}
  n18 = LinearInterpolate(
   A=LinearInterpolate(Alpha=MaterialFunctionCall!(PrimaryTextureCoordinates=@n16, 2ndOrderTextureCoordinates=@n16, RandomSeed=@n13, NoiseDensityBias, Bypass Noise Influence Crosstalk) {materialFunction=/Game/Effects/Fort_Effects/Materials/_Functions/MF_RandomUVSample}.NoiseWithRange) {constA=0.5}
   B=LinearInterpolate(Alpha=Saturate(None=@n17.Raw Flicker)) {constA=0.5}
   Alpha=ScalarParameter() {defaultValue=1, parameterName=Use Hex Random Pattern?}
  )
  n19 = NamedRerouteUsage()
  n20 = Panner(
   Coordinate=Multiply(A=@n19, B=ScalarParameter() {defaultValue=0.25, parameterName=Backlight Noise Tile}) {constB=0.25}
   Speed=LinearInterpolate(
    A=Constant2Vector() {g=0.01, r=-0.01}
    B=Constant2Vector() {g=-0.01, r=0.01}
    Alpha=MaterialFunctionCall(Value=@n1) {materialFunction=&LinearSine}.Linear Sine
   )
  ) {speedX=-0.1, speedY=0.1}
  n21 = Multiply(A=TextureSample(UVs=Panner(Coordinate=@n19) {speedY=-0.02}) {texture=&2ndInfectedMapLinear}.R) {constB=0.025}
  n22 = Power(Base=TextureSample(UVs=Add(A=@n20, B=@n21) {desc=Distort UVs}) {samplerType=SAMPLERTYPE_LinearGrayscale, texture=/Game/Global/Textures/Default/Noise/LowResBlurredNoise_NoMips}.G)
  Multiply(A=LinearInterpolate(Alpha=Saturate(None=@n22)) {constA=0.1, constB=1.5}, B=@n18)
  Multiply(A=@n7) {constB=0.05}
  n23 = Constant() {r=5}
  n24 = Multiply(A=Power(Base=MaterialFunctionCall!(TexCoords, Edge Sharpness=@n23, Edge Subtract) {materialFunction=/Game/Effects/Niagara/MasterMaterials/Functions/Mojo_MF/MF_FX_UV_EdgeMask}.1-Edge Mask) {constExponent=5.5}) {constB=1.15}
  Multiply(A=@n10, B=@n24) {desc=Vignette}
  NamedRerouteDeclaration(None=LinearInterpolate(A=OneMinus(None=@n5), B=@n5, Alpha=OneMinus(None=@n11))) {name=Scanlines_A, nodeColor=<rgba 1 0 0.7266 1>}
  NamedRerouteDeclaration(None=Saturate(None=@n24)) {name=EdgeMask, nodeColor=<rgba 1 0 0.1406 1>}
  n25 = VectorParameter() {parameterName=Color1, defaultValue=<rgba 0.0115 0.0184 0.1875 1>}
  n26 = VectorParameter() {parameterName=Color2, defaultValue=<rgba 0.0435 0.0174 0.2865 1>}
  n27 = NamedRerouteUsage()
  NamedRerouteDeclaration(None=LinearInterpolate(A=LinearInterpolate(A=@n25.RGB, B=@n26.RGB, Alpha=@n12), B=LinearInterpolate(A=@n26.RGB, B=@n25.RGB, Alpha=@n27), Alpha=@n5)) {name=PrimaryScreenColor, nodeColor=<rgba 1 0 0.8672 1>}
  NamedRerouteDeclaration(None=LinearInterpolate(Alpha=@n18) {constA=0.1, constB=0.8}) {name=Random Square or Hex Patterns, desc=Swapped squares to Hex, nodeColor=<rgba 0 0.1328 1 1>}
  NamedRerouteDeclaration(None=@n15.9-Slice UV) {name=uv, nodeColor=<rgba 0.875 1 0 1>}
  NamedRerouteDeclaration(None=TextureSample(UVs=Multiply(A=@n15.9-Slice UV, B=@n14)) {samplerType=SAMPLERTYPE_Masks, texture=/Game/Environments/Asteria/Sets/NeonCity/Techart/Textures/T_LCD_Panels_LCDSoft}.RGB) {name=LCD pattern, nodeColor=<rgba 0.5703 1 0 1>}
  NamedRerouteUsage()
  Reroute(None=@n10)
  Saturate(None=@n10)
  TextureCoordinate()
  TextureCoordinate()
  TextureCoordinate()
  TextureCoordinate() {uTiling=0.5, vTiling=0.5}
  TextureCoordinate() {uTiling=0.95, vTiling=0.95}
  TextureCoordinate()
 ]
}
